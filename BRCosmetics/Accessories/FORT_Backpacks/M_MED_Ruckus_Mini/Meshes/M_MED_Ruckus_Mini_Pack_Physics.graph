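# Physics asset for the M_MED_Ruckus_Mini backpack cosmetic, defining collision bodies and joints for its dangling parts.
# A capsule on the pack bone anchors separately constrained zipper, handle and cube bodies; the cube uses a box, the rest capsules.
# The flap is a four-capsule chain from the pack through dyn_flap_MSTR to dyn_flap_03.
# Tight swing limits and little or no twist restrict each part to small secondary motion.
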PhysicsAsset /BRCosmetics/Accessories/FORT_Backpacks/M_MED_Ruckus_Mini/Meshes/M_MED_Ruckus_Mini_Pack_Physics
bodies {
 dyn_cube [{
   shapeType=Box
   extentX=11.454
   extentY=11.8528
   extentZ=10.1028
   shapeName=dyn_cube_box
   center=<4.0681 -0.6365 -0.3725>
   rotation=<rot 8.7518 -100.7179 -84.698>
  }]
 dyn_flap_01 [{shapeType=Capsule, length=15.8323, radius=1.55, center=<1 0 0>, rotation=<rot 1.0464 0 0>}]
 dyn_flap_02 [{shapeType=Capsule, length=15, radius=2, shapeName=dyn_flap_02_capsule, center=<1 0 0>, rotation=<rot 0.1235 0 0>}]
 dyn_flap_03 [{shapeType=Capsule, length=13.9923, radius=1.75, shapeName=dyn_flap_03_capsule, center=<1 0 0>, rotation=<rot 2.9541 0 0>}]
 dyn_flap_MSTR [{shapeType=Capsule, length=18.2131, radius=1.55, shapeName=dyn_flap_MSTR_capsule, center=<1 0 0>}]
 dyn_handle [{
   shapeType=Capsule
   length=3.7246
   radius=2.0642
   shapeName=dyn_handle_capsule
   center=<2.3167 -0.0035 -0.0922>
   rotation=<rot 0.0117 -89.9999 1.8792>
  }]
 dyn_zipper [{shapeType=Capsule, length=1.76, radius=0.3, shapeName=dyn_zipper_capsule, center=<1.2156 0.0649 -0.0149>, rotation=<rot 0 90.3112 -89.3381>}]
 pack [{shapeType=Capsule, length=15, radius=7.25, shapeName=pack_capsule, center=<-0.214 -15.3547 -12.5368>, rotation=<rot 58.2544 81.7391 80.866>}]
}
constraints [
 {bone1Name=dyn_flap_01, bone2Name=dyn_flap_MSTR, swing1LimitDegrees=8, swing2LimitDegrees=5, twistLimitDegrees=0}
 {bone1Name=dyn_flap_02, bone2Name=dyn_flap_01, swing1LimitDegrees=8, swing2LimitDegrees=5, twistLimitDegrees=0}
 {bone1Name=dyn_flap_03, bone2Name=dyn_flap_02, swing1LimitDegrees=5, swing2LimitDegrees=2, twistLimitDegrees=0}
 {bone1Name=dyn_zipper, bone2Name=pack, swing1LimitDegrees=40, swing2LimitDegrees=1.35, twistLimitDegrees=0.01}
 {bone1Name=dyn_handle, bone2Name=pack, swing1LimitDegrees=20, swing2LimitDegrees=10, twistLimitDegrees=2}
 {bone1Name=dyn_cube, bone2Name=pack, swing1LimitDegrees=26, swing2LimitDegrees=6, twistLimitDegrees=2}
 {bone1Name=dyn_flap_MSTR, bone2Name=pack, swing1LimitDegrees=1, swing2LimitDegrees=1, twistLimitDegrees=0}
]
