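# Animation Blueprint for the M_MED_Ruckus_Mini backpack cosmetic that adds physics motion to the pack.
# AnimGraph applies a small additive roll to the pack bone, then feeds a RigidBody node simulating in root-bone space with the pack's physics asset and overridden gravity.
# Each update sets Aiming from bIsTargeting, which drives the simulation alpha slightly higher while targeting.
# A second RigidBody node based on the pack bone is left unconnected.
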AnimBlueprint /BRCosmetics/Accessories/FORT_Backpacks/M_MED_Ruckus_Mini/Meshes/M_MED_Ruckus_Mini_Pack_AnimBP
target_skeleton=&/M_MED_Ruckus_Mini_Pack_Skeleton
graphs [
 AnimGraph {nodes [
   RigidBody() {
    bClampLinearTranslationLimitToRefPose=true
    bOverrideWorldGravity=true
    overridePhysicsAsset=&/M_MED_Ruckus_Mini_Pack_Physics
    simulationSpace=BaseBoneSpace
    alphaBoolBlend {blendOption=Linear}
    baseBoneRef {boneName=pack}
    componentAppliedLinearAccClamp=<3000 3000 3000>
    componentLinearAccScale=<0.4 0.4 0.4>
    componentLinearVelScale=<0.3 0.3 0.3>
    overrideWorldGravity {z=-980}
   }
   n1 = "Transform (Modify) Bone - Bone: pack"(ComponentPose=Local To Component(LocalPose=Input Pose()), Translation="0,0,0", Rotation="0,0,-6") {
    rotationMode=BMM_Additive
    translationSpace=BCS_ParentBoneSpace
    alphaBoolBlend {blendOption=Linear}
    boneToModify {boneName=pack}
    rotation {roll=-6}
   }
   n2 = RigidBody(ComponentPose=@n1, Alpha=Get Aiming() -> (Aiming:double)) {
    bClampLinearTranslationLimitToRefPose=true
    bOverrideWorldGravity=true
    overridePhysicsAsset=&M_MED_Ruckus_Mini_Pack_Physics
    simulationSpace=BaseBoneSpace
    alphaBoolBlend {blendOption=Linear}
    baseBoneRef {boneName=root}
    componentAppliedLinearAccClamp=<1500 1500 1500>
    componentLinearAccScale=<0.6 0.6 0.6>
    componentLinearVelScale=<0.4 0.4 0.4>
    overrideWorldGravity {z=-980}
   }
   Output Pose(Result=Component To Local(ComponentPose=@n2)) {layerGroup=None}
  ]}
 EventGraph {nodes [
   Event BlueprintUpdateAnimation()
   > Set Aiming(Aiming:double=SelectFloat(A=0.8, B=0.65, bPickA=Get bIsTargeting())) -> (Output_Get:double)
   TryGetPawnOwner()
  ]}
]
variables [Aiming {type=number, default=0.2}]
